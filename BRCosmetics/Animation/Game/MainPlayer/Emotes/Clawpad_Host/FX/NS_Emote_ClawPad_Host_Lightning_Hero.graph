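# Niagara effect for the ClawPad Host emote: flickering lightning meshes from skeletal mesh bones or sockets, with impacts, sparks and smoke.
# Main_Lightning fires staggered bursts over three loops; Small_Branches and GroundImpactsLightning read its particles and offset along bolt orientation.
# Directional_sparks spawn from those impacts, chaining emitters through particle reads.
# Smoke runs separately with pulsed spawning and light flashes, scaled by front-end context, its sprites also tinted by time of day.
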NiagaraSystem /BRCosmetics/Animation/Game/MainPlayer/Emotes/Clawpad_Host/FX/NS_Emote_ClawPad_Host_Lightning_Hero
dependencies {
 usedDataInterfaces [
  &NiagaraDataInterfaceCurve
  &NiagaraDataInterfaceArrayInt32
  &NiagaraDataInterfaceParticleRead
  &NiagaraDataInterfaceVector2DCurve
  &NiagaraDataInterfaceSpriteRendererInfo
  &NiagaraDataInterfaceColorCurve
  &NiagaraDataInterfaceSkeletalMesh
 ]
 usedDynamicInputs [
  &MakeLinearColorFromVectorAndFloat
  &MakeCustomVectorFromBool
  &SetBoolByFloatComparison
  &RandomRangeVector2D
 ]
 usedModules [
  &self:SystemSpawnScript.SystemScriptSource.SystemScriptGraph.NiagaraNodeAssignment_1.SetVariables_#1
  &self:SystemSpawnScript.SystemScriptSource.SystemScriptGraph.NiagaraNodeAssignment_0.SetVariables_#2
  &self:SystemSpawnScript.SystemScriptSource.SystemScriptGraph.NiagaraNodeAssignment_2.SetVariables_#3
  &SystemState
 ]
}
emitters {
 Directional_sparks {
  emitterUpdateScript [
   &EmitterState(Life Cycle Mode=::NewEnumerator1, Loop Behavior=::NewEnumerator1, Loop Duration=4, UseLoopDelay=false)
   &SpawnParticlesFromOtherEmitter(
    Attribute Reader=DataInterface() {propertyValues {EmitterBinding {bindingMode=Other, emitterName=GroundImpactsLightning}}}
    Spawn Rate=30
    Use Maximum Age=false
    Use Minimum Age=false
   )
  ]
  particleSpawnScript [
   &InitializeParticle(
    Color=$System.PalpatineColor
    Lifetime Max=0.2
    Lifetime Min=0.4
    Lifetime Mode=::NewEnumerator1
    Sprite Size Max=<2 5>
    Sprite Size Min=<1 2>
    Sprite Size Mode=::NewEnumerator2
    UsePositionOffset=false
    Write Material Random=true
   ) {module=InitializeParticle001}
   &ShapeLocation(
    Cylinder Height=10
    Cylinder Radius=2
    MaskSpawnGroup=false
    Override Seed=false
    Shape Primitive=::NewEnumerator1
    Surface Only=false
    UseOwnerScale=true
   ) {module=ShapeLocation001}
   &CameraOffset(Camera Offset Amount=15)
   &AddVelocity(
    Cone Angle=200
    Cone Axis=<0 0 1>
    Distribution Along Cone Axis=90
    Module.UseConeFalloff=true
    Module.UseVelDistribution=true
    Override Seed=false
    Speed Falloff From Cone Axis=0.5
    Velocity Mode=::NewEnumerator2
    Velocity Speed=&RandomRangeFloat(Minimum=100, Maximum=300)
   ) {module=AddVelocity001}
   &SampleParticlesFromOtherEmitter(Maximum Sampled Particle Age=0.03, Module.MaxSampledAgeCutoff=true, Module.MinSampledAgeCutoff=false, Reject Particle By Sampled Age=true)
  ]
  particleUpdateScript [
   &ParticleState()
   &Drag(Drag=&RandomRangeFloat(Minimum=2, Maximum=3), Use Linear Drag=true, Use Rotational Drag=false)
   &GravityForce(Gravity=&RandomRangeVector(Minimum=<0 0 -980>, Maximum=<0 0 -500>))
   &SolveForcesAndVelocity(Clamp Velocity=false, Limit Acceleration=false)
   &ScaleSpriteSizeBySpeed(Max Scale Factor=<1 10>, Min Scale Factor=<0.65 2>)
   &ScaleSpriteSize(Scale Factor=&Vector2DFromCurve(Vector2Curve=DataInterface() {propertyValues {XCurve {keys [<at 0 1>,<at 1 1>]}, YCurve {keys [<at 0 1>,<at 1 0>]}}}, ScaleCurve=<1 1>))
   &ScaleColor(Scale RGB=&Multiply_VectorByFloat(Vector=<1 1 1>, Float=10), ScaleA=true, ScaleRGB=true)
  ]
  rendererClasses [&NiagaraSpriteRendererProperties]
  renderers [{rendererClass=&NiagaraSpriteRendererProperties}]
 }
 GroundImpactsLightning {
  emitterUpdateScript [
   &EmitterState(
    Life Cycle Mode=::NewEnumerator1
    Loop Behavior=::NewEnumerator2
    Loop Count=3
    Loop Delay=0.15
    Loop Duration=1
    Recalculate Duration Each Loop=true
    UseLoopDelay=false
   )
   &SpawnParticlesFromOtherEmitter(
    Attribute Reader=DataInterface() {propertyValues {EmitterBinding {bindingMode=Other, emitterName=Main_Lightning}}}
    Spawn Group:NiagaraInt32=&SelectIntFromArray(
     Array Sampling Mode=::NewEnumerator1
     Direct Array Index=?
     Evaluation Type=::NewEnumerator1
     Int Selection Array=DataInterface() {propertyValues {IntData [1,1,1,1,1,1,0,0,0,0,0,0,0]}}
     Random Seed=0
    )
    Spawn Rate=15
    Use Maximum Age=false
    Use Minimum Age=false
   )
  ]
  particleSpawnScript [
   /SetVariables(Particles.VisibilityTag=&SpawnGroup()) {isSetParameters=true}
   &InitializeParticle(
    Color=&Multiply_LinearColorByFloat(Linear Color=$System.PalpatineColor, Float=1.5)
    Lifetime Max=0.07
    Lifetime Min=0.3
    Lifetime Mode=::NewEnumerator1
    Mesh Scale Max=<2.5 2.5 4>
    Mesh Scale Min=<2 2 1.5>
    Mesh Scale Mode=::NewEnumerator2
    UsePositionOffset=false
    Write Material Random=true
   )
   &SampleParticlesFromOtherEmitter(
    Maximum Sampled Particle Age:NiagaraFloat=&FloatFromCurve(
     FloatCurve=DataInterface() {propertyValues {Curve {keys [<at 0 0.5>,<at 1.7 0.5>,<at 1.8 1>,<at 2 1>]}}}
     CurveIndex=$System.Age
     Scale Curve=0.9
    )
    Mesh Orientation Sampling:ENiagara_DirectReadSamplingMode=::NewEnumerator0
    Minimum Sampled Particle Age=0.01
    Module.MaxSampledAgeCutoff=true
    Module.MinSampledAgeCutoff=true
    Position Sampling:ENiagara_DirectReadSamplingMode=::NewEnumerator1
    Reject Particle By Sampled Age=true
   )
   /SetVariables(Particles.BasePosition:NiagaraPosition=&AddVectorToPosition(
     Position=$Output.SampleParticlesFromOtherEmitter.SampledPosition
     Vector=&Multiply_VectorByFloat(
      Vector=&ReturnMeshOrientationAxis(Mesh Orientation Axis=::ZAxis, Mesh Orientation Quat=$Particles.MeshOrientation)
      Float=&MaskFloat(Mask Source=0, Mask Parameter=?, Mask=0, Masked Float=117, Passthrough Float=120)
     )
    )) {isSetParameters=true}
   &InitialMeshOrientation(Mesh Orientation Mode=::NewEnumerator2, Rotation=<0.5 0 0>, Use Rotation Vector=true)
  ]
  particleUpdateScript [
   &SolveForcesAndVelocity(Clamp Velocity=false, Limit Acceleration=false)
   &ShapeLocation(
    MaskSpawnGroup=false
    Offset Mode=::NewEnumerator0
    Override Seed=false
    Shape Origin=$Particles.BasePosition
    Sphere Radius=1
    Sphere Surface Distribution=1
    UseOwnerScale=true
   )
   &ParticleState()
   &UpdateMeshOrientation(
    Facing Direction=&Add_Vector(A=&Multiply_Vector(A=$Particles.ShapeLocation.ShapeNormal, B=$Engine.Owner.SystemZAxis), B=<0 0 0>)
    Facing Interpolation=false
    Orientation Method=::NewEnumerator1
    Override Quat=false
    Side Direction:Vector3f=<0 0 -1>
    Stabilization Mode:ENiagara_OrientVectorCount=::NewEnumerator3
    TransformFacing=false
    TransformSide=false
    TransformUp=false
    Up Direction:Vector3f=<0 0 0>
   )
   &ScaleMeshSize(Scale Factor=&VectorFromFloat(Value=0.1))
   &DynamicMaterialParameters(
    Index 0 Param 1=&MakeFloatFromVector(Vector=$Particles.Scale, Channel=::NewEnumerator0)
    Index 0 Param 2=&RandomRangeFloat(Minimum=0, Maximum=10)
    Index 0 Param 3=&FloatFromCurve(
     FloatCurve=DataInterface() {propertyValues {Curve {keys [{interpMode=RCIM_Cubic, leaveTangent=8.1085, tangentMode=RCTM_Break, time=0.8568, value=1}, {tangentMode=RCTM_User, time=1, value=0}]}}}
     Scale Curve=0.015
    )
    Index 0 Param 4=0.8
    Index 1 Param 1=&FloatFromCurve(FloatCurve=DataInterface() {propertyValues {Curve {keys [
         {interpMode=RCIM_Cubic, leaveTangent=1.1034, tangentMode=RCTM_Break, time=0, value=0}
         {arriveTangent=2.2399, tangentMode=RCTM_User, time=0.5, value=1}
        ]}}})
    Index 1 Param 2=&FloatFromCurve(FloatCurve=DataInterface() {propertyValues {Curve {keys [<at 0.5 0>, {arriveTangent=3.2053, tangentMode=RCTM_User, time=0.9, value=1}]}}})
    Index 1 Param 3=&FloatFromCurve(FloatCurve=DataInterface() {propertyValues {Curve {keys [<at 0 0 RCIM_Cubic>, {arriveTangent=3.4223, interpMode=RCIM_Cubic, leaveTangent=3.4223, tangentMode=RCTM_User, time=1, value=1}]}}})
    Index 1 Param 4=&FloatFromCurve(
     FloatCurve=DataInterface() {propertyValues {Curve {keys [
         {interpMode=RCIM_Cubic, leaveTangent=0.1169, tangentMode=RCTM_Break, time=0, value=0}
         {arriveTangent=2.5708, interpMode=RCIM_Cubic, leaveTangent=2.5708, tangentMode=RCTM_User, time=1, value=1}
        ]}}}
     Scale Curve=&RandomRangeFloat(Minimum=1, Maximum=40)
    )
    Index 2 Param 1=&FloatFromCurve(FloatCurve=DataInterface() {propertyValues {Curve {keys [
         {interpMode=RCIM_Cubic, leaveTangent=-0.0025, tangentMode=RCTM_Break, time=0, value=0.5}
         {arriveTangent=-0.0316, interpMode=RCIM_Cubic, tangentMode=RCTM_Break, time=1, value=-0.2}
        ]}}})
    Index 2 Param 2=&FloatFromCurve(
     FloatCurve=DataInterface() {propertyValues {Curve {keys [
         {interpMode=RCIM_Cubic, tangentMode=RCTM_User, time=0, value=0}
         {arriveTangent=2.0522, interpMode=RCIM_Cubic, leaveTangent=2.0522, tangentMode=RCTM_User, time=1, value=1}
        ]}}}
     Scale Curve=2
    )
    Module.Index0Param0Write=true
    Module.Index0Param1Write=true
    Module.Index0Param2Write=true
    Module.Index0Param3Write=true
    Module.Index1Param0Write=true
    Module.Index1Param1Write=true
    Module.Index1Param2Write=true
    Module.Index1Param3Write=true
    Module.Index2Param0Write=true
    Module.Index2Param1Write=true
    Module.Index2Param2Write=false
    Module.Index2Param3Write=false
    Module.Index3Param0Write=false
    Module.Index3Param1Write=false
    Module.Index3Param2Write=false
    Module.Index3Param3Write=true
    Write Parameter Index 1=true
    Write Parameter Index 2=true
    Write Parameter Index 3=true
   )
   &ScaleColor(
    Scale Alpha=&FloatFromCurve(FloatCurve=DataInterface() {propertyValues {Curve {keys [
         {tangentMode=RCTM_User, time=0, value=1}
         {arriveTangent=-0.0405, interpMode=RCIM_Cubic, leaveTangent=-0.0405, tangentMode=RCTM_User, time=0.7009, value=1}
         {arriveTangent=-8.0089, interpMode=RCIM_Cubic, leaveTangent=-8.0089, tangentMode=RCTM_User, time=1, value=0}
        ]}}})
    Scale RGB=&Lerp_Vector(
     A=&VectorFromFloat(Value=&FloatFromCurve(
       FloatCurve=DataInterface() {propertyValues {Curve {keys [<at 0 1>,<at 0.1 1>,<at 0.2 2>,<at 0.3 1>,<at 0.6 1>,<at 0.7 2>,<at 0.8 1>,<at 1 1>]}}}
       Scale Curve=10
      ))
     B=&VectorFromFloat(Value=&RandomRangeFloat(Minimum=0.3, Maximum=5, Evaluation Type=::NewEnumerator1))
     Alpha=&FloatFromCurve(FloatCurve=DataInterface() {propertyValues {Curve {keys [
          {interpMode=RCIM_Cubic, leaveTangent=2.5294, tangentMode=RCTM_Break, time=0, value=0}
          {arriveTangent=-0.2166, interpMode=RCIM_Cubic, tangentMode=RCTM_Break, time=0.4, value=1}
         ]}}})
    )
    ScaleA=true
    ScaleRGB=false
   )
   &CameraOffset(Camera Offset Amount=&MaskFloat(
     Mask Source=0
     Mask Parameter=?
     Mask=0
     Masked Float=&RandomRangeFloat(Minimum=5, Maximum=10)
     Passthrough Float=10
    ))
  ]
  rendererClasses [&NiagaraMeshRendererProperties, &NiagaraSpriteRendererProperties]
  renderers [
   {rendererClass=&NiagaraMeshRendererProperties, rendererIndex=0}
   {rendererClass=&NiagaraMeshRendererProperties, rendererIndex=1}
   {rendererClass=&NiagaraSpriteRendererProperties, rendererIndex=2}
  ]
 }
 Main_Lightning {
  emitterUpdateScript [
   &EmitterState(
    Delay First Loop Only=true
    Life Cycle Mode=::NewEnumerator1
    Loop Behavior=::NewEnumerator2
    Loop Count=3
    Loop Delay=0.15
    Loop Duration=&FloatFromCurve(
     FloatCurve=DataInterface() {propertyValues {Curve {keys [
         <at 0 1>
         {arriveTangent=0.8333, interpMode=RCIM_Cubic, leaveTangent=0.8333, time=0.5, value=1}
         <at 0.6 1.5 RCIM_Constant>
         <at 1 1.5 RCIM_Cubic>
        ]}}}
     CurveIndex=$Engine.System.Age
     Scale Curve=0.9
    )
    Recalculate Duration Each Loop=true
    UseLoopDelay=true
   )
   /SetVariables(Emitter.ColorCoreScalar=10) {isSetParameters=true}
   &SpawnBurst_Instantaneous(Spawn Count=5, Use Spawn Probability=false, UseLoopCountLimit=false)
   &SpawnBurst_Instantaneous(Spawn Count=2, Spawn Time=0.08, Use Spawn Probability=false, UseLoopCountLimit=false) {module=SpawnBurst_Instantaneous001}
   &SpawnBurst_Instantaneous(Spawn Time=0.16, Use Spawn Probability=false, UseLoopCountLimit=false) {module=SpawnBurst_Instantaneous002}
   &SpawnBurst_Instantaneous(Spawn Count=2, Spawn Time=0.24, Use Spawn Probability=false, UseLoopCountLimit=false) {module=SpawnBurst_Instantaneous003}
  ]
  particleSpawnScript [
   &PartitionParticles(Number Of Partitions=10, Particles Per Partition=1, Partition Mode:ENiagara_PartitionMode_v2=::NewEnumerator1)
   /SetVariables(Particles.GlobalRandom:NiagaraFloat=&RandomRangeFloat(Minimum=0.4, Maximum=1)) {isSetParameters=true}
   &InitializeParticle(
    Color=$System.PalpatineColor
    Lifetime=&Multiply_Float(
     A=&RandomRangeFloat(Minimum=0.4, Maximum=0.7)
     B=&FloatFromCurve(
      FloatCurve=DataInterface() {propertyValues {Curve {keys [<at 0 0.5>,<at 1.7 0.5>,<at 1.8 1>,<at 2 1>]}}}
      CurveIndex=$System.Age
      Scale Curve=2
     )
    )
    Mesh Scale Mode=::NewEnumerator3
    Mesh Uniform Scale=&MakeFloatFromVector(
     Vector=&MakeVector(X=1, Y=1, Z=&Multiply_Float(A=0.3, B=$Particles.GlobalRandom))
     Channel=::NewEnumerator0
    )
    Position Mode=::NewEnumerator0
    Sprite Size Mode=::NewEnumerator3
    Uniform Sprite Size=5
    Write Material Random=true
   )
   &InitialMeshOrientation(
    Mesh Axis To Orient=&MaskVector(Mask Source=1, Mask Parameter=$Particles.PartitionParticles.Partition, Mask=0, Masked Vector=<0 0 -1>, Passthrough Vector=<0 0 -1>)
    Mesh Orientation Mode=::NewEnumerator2
    Orientation Vector=<0 0 1>
    Rotation=&MaskVector(
     Mask Source=1
     Mask Parameter=$Particles.PartitionParticles.Partition
     Mask=0
     Masked Vector=<-0.0026 -0.0221 0>
     Passthrough Vector=<0.0051 -0.0008 0>
    )
    Rotation Coordinate Space=::NewEnumerator2
    Use Rotation Vector=true
   )
  ]
  particleUpdateScript [
   &ParticleState()
   &SkeletalMeshLocation(
    Bone / Socket Index=&SelectIntFromArray(
     Direct Array Index=$Particles.PartitionParticles.Partition
     Int Selection Array=DataInterface() {propertyValues {IntData [0,5,1,7,3,9,4,2,8,6]}}
    )
    Bone Sampling Mode=::NewEnumerator4
    MaskSpawnGroup=false
    Orientation Sampling=::NewEnumerator1
    Skeletal Mesh=$System.Skeletal Mesh
   )
   &ShapeLocation(
    MaskSpawnGroup=false
    Override Seed=false
    Rotation Coordinate Space=::Simulation
    Shape Origin=$Particles.SkeletalMeshLocation.SampledPosition
    Sphere Radius=0.01
    Sphere Surface Distribution=1
    UseOwnerScale=true
   )
   &UpdateMeshOrientation(
    Facing Direction=&TransformVector(
     Vector=&Add_Vector(A=&Multiply_Vector(A=$Particles.ShapeLocation.ShapeNormal, B=<0.5 0.3 0.3>), B=$Engine.Owner.SystemXAxis)
     Source Space=::World
     Destination Space=::Local
    )
    Facing Interpolation=false
    Orientation Method=::NewEnumerator1
    Override Quat=false
    TransformFacing=false
    TransformUp=false
   )
   &SolveForcesAndVelocity(Clamp Velocity=false, Limit Acceleration=false)
   &DynamicMaterialParameters(
    Index 0 Param 1=&FloatFromCurve(
     FloatCurve=DataInterface() {propertyValues {Curve {
        defaultValue=0
        keys [
         {arriveTangent=9.0725, interpMode=RCIM_Cubic, leaveTangent=9.0725, tangentMode=RCTM_User, time=0, value=0}
         {arriveTangent=-2.5249, interpMode=RCIM_Cubic, leaveTangent=-2.5249, tangentMode=RCTM_User, time=0.1408, value=0.9847}
         {arriveTangent=-0.0405, interpMode=RCIM_Cubic, leaveTangent=-0.0405, tangentMode=RCTM_User, time=1, value=0.2}
        ]
       }}}
     Scale Curve=&RandomRangeFloat(Minimum=1, Maximum=0.5)
    )
    Index 0 Param 2=&FloatFromCurve(
     FloatCurve=DataInterface() {propertyValues {Curve {
        defaultValue=0
        keys [<at 0 0.2 RCIM_Cubic>, {arriveTangent=2.4937, interpMode=RCIM_Cubic, leaveTangent=2.4937, tangentMode=RCTM_User, time=1, value=1}]
       }}}
     Scale Curve=&Multiply_Float(A=30, B=$Particles.GlobalRandom)
    )
    Index 0 Param 3=&RandomRangeFloat(Minimum=0.25, Maximum=1)
    Index 0 Param 4=&FloatFromCurve(FloatCurve=DataInterface() {propertyValues {Curve {
        defaultValue=0
        keys [
         {arriveTangent=4.1785, interpMode=RCIM_Cubic, leaveTangent=4.1785, tangentMode=RCTM_User, time=0, value=0}
         {arriveTangent=0.0224, interpMode=RCIM_Cubic, leaveTangent=0.0224, tangentMode=RCTM_User, time=0.5, value=1}
        ]
       }}})
    Index 1 Param 1=&FloatFromCurve(FloatCurve=DataInterface() {propertyValues {Curve {
        defaultValue=1
        keys [
         {interpMode=RCIM_Cubic, tangentMode=RCTM_User, time=0, value=0}
         {arriveTangent=2.9607, interpMode=RCIM_Cubic, leaveTangent=2.9607, tangentMode=RCTM_User, time=0.7, value=1}
        ]
       }}})
    Index 1 Param 2=&Multiply_Float(A=1, B=$Particles.GlobalRandom)
    Index 1 Param 3=&Multiply_Float(A=0.5, B=$Particles.GlobalRandom)
    Index 1 Param 4=&RandomRangeFloat(Minimum=0.5, Maximum=1)
    Module.Index0Float0Write=true
    Module.Index0Float1Write=true
    Module.Index0Float2Write=true
    Module.Index0Float3Write=true
    Module.Index1Float0Write=true
    Module.Index1Float1Write=true
    Module.Index1Float2Write=true
    Module.Index1Float3Write=true
    Write Parameter Index 1=::NewEnumerator1
   )
   &ScaleColor(
    Scale RGB=&Multiply_VectorByFloat(
     Vector=<1 1 1>
     Float=&Waveform(Global Phase=$Particles.Age, "Waveform [1]"=::NewEnumerator0, "[1] Frequency"=10, "[1] Amplitude Min/Max"=<2 0.3>)
    )
    ScaleA=true
    ScaleRGB=true
   )
  ]
  rendererClasses [&NiagaraMeshRendererProperties]
  renderers [{rendererClass=&NiagaraMeshRendererProperties, rendererIndex=0}]
 }
 Small_Branches {
  emitterUpdateScript [
   &EmitterState(
    Delay First Loop Only=true
    Life Cycle Mode=::NewEnumerator1
    Loop Behavior=::NewEnumerator2
    Loop Count=3
    Loop Delay=0.15
    Loop Duration=1
    Recalculate Duration Each Loop=true
    UseLoopDelay=true
   )
   &SpawnParticlesFromOtherEmitter(
    Attribute Reader=DataInterface() {propertyValues {EmitterBinding {bindingMode=Other, emitterName=Main_Lightning}}}
    Use Maximum Age=false
    Use Minimum Age=false
   )
  ]
  particleSpawnScript [
   &PartitionParticles()
   &InitializeParticle(
    Color=&Multiply_LinearColorByFloat(Linear Color=$System.PalpatineColor, Float=1)
    Lifetime Max=0.2
    Lifetime Min=0.3
    Lifetime Mode=::NewEnumerator1
    Mesh Scale Max=<1 1 3>
    Mesh Scale Mode=::NewEnumerator2
    UsePositionOffset=false
    Write Material Random=true
   )
   &InitialMeshOrientation(
    Mesh Orientation Mode=::NewEnumerator2
    Rotation=&RandomRangeVector(Minimum=<0.55 0 0>, Maximum=<0.45 0 0>)
    Use Rotation Vector=true
   )
   &SampleParticlesFromOtherEmitter(
    Maximum Sampled Particle Age=0.25
    Mesh Orientation Sampling:ENiagara_DirectReadSamplingMode=::NewEnumerator1
    Module.MaxSampledAgeCutoff=true
    Module.MinSampledAgeCutoff=false
    Position Sampling:ENiagara_DirectReadSamplingMode=::NewEnumerator1
    Reject Particle By Sampled Age=true
   )
   /SetVariables(Particles.BasePosition:NiagaraPosition=&AddVectorToPosition(
     Position=$Output.SampleParticlesFromOtherEmitter.SampledPosition
     Vector=&Multiply_VectorByFloat(
      Vector=&ReturnMeshOrientationAxis(Mesh Orientation Axis=::ZAxis, Mesh Orientation Quat=$Particles.MeshOrientation)
      Float=&RandomRangeFloat(Minimum=50, Maximum=80)
     )
    )) {isSetParameters=true}
  ]
  particleUpdateScript [
   &SolveForcesAndVelocity(Clamp Velocity=false, Limit Acceleration=false)
   &SkeletalMeshLocation(
    Bone / Socket Index=$Particles.PartitionParticles.Partition
    Bone Sampling Mode=::NewEnumerator4
    MaskSpawnGroup=false
    Orientation Sampling=::NewEnumerator1
    Skeletal Mesh=$System.Skeletal Mesh
   )
   &ShapeLocation(
    MaskSpawnGroup=false
    Override Seed=false
    Shape Origin=$StackContext.SkeletalMeshLocation.SampledPosition
    Sphere Radius=0.02
    Sphere Surface Distribution=1
    UseOwnerScale=true
   )
   &ParticleState()
   &UpdateMeshOrientation(
    Facing Direction=&Add_Vector(A=&Multiply_Vector(A=$Particles.ShapeLocation.ShapeNormal, B=<1 1 1.0528>), B=$Engine.Owner.SystemYAxis)
    Facing Interpolation=false
    Orientation Method=::NewEnumerator1
    Override Quat=false
    TransformFacing=false
    TransformUp=false
    Up Direction:Vector3f=&RotateVectorByQuaternion(Vector To Rotate=<0 0 1>, Quaternion=$Output.SampleParticlesFromOtherEmitter.SampledMeshOrientation)
   )
   &ScaleMeshSize(Scale Factor=&VectorFromFloat(Value=0.1))
   &DynamicMaterialParameters(
    Index 0 Param 1=&MakeFloatFromVector(Vector=$Particles.Scale, Channel=::NewEnumerator0)
    Index 0 Param 2=&RandomRangeFloat(Minimum=0, Maximum=10)
    Index 0 Param 3=&FloatFromCurve(
     FloatCurve=DataInterface() {propertyValues {Curve {keys [{interpMode=RCIM_Cubic, leaveTangent=8.1085, tangentMode=RCTM_Break, time=0.8568, value=1}, {tangentMode=RCTM_User, time=1, value=0}]}}}
     Scale Curve=0.03
    )
    Index 0 Param 4=0.8
    Index 1 Param 1=&FloatFromCurve(FloatCurve=DataInterface() {propertyValues {Curve {keys [
         {interpMode=RCIM_Cubic, leaveTangent=1.1034, tangentMode=RCTM_Break, time=0, value=0}
         {arriveTangent=2.2399, tangentMode=RCTM_User, time=0.5, value=1}
        ]}}})
    Index 1 Param 2=&FloatFromCurve(FloatCurve=DataInterface() {propertyValues {Curve {keys [<at 0.5 0>, {arriveTangent=3.2053, tangentMode=RCTM_User, time=0.9, value=1}]}}})
    Index 1 Param 3=&FloatFromCurve(FloatCurve=DataInterface() {propertyValues {Curve {keys [<at 0 0 RCIM_Cubic>, {arriveTangent=3.4223, interpMode=RCIM_Cubic, leaveTangent=3.4223, tangentMode=RCTM_User, time=1, value=1}]}}})
    Index 1 Param 4=&FloatFromCurve(
     FloatCurve=DataInterface() {propertyValues {Curve {keys [
         {interpMode=RCIM_Cubic, leaveTangent=0.1169, tangentMode=RCTM_Break, time=0, value=0}
         {arriveTangent=2.5708, interpMode=RCIM_Cubic, leaveTangent=2.5708, tangentMode=RCTM_User, time=1, value=1}
        ]}}}
     Scale Curve=&RandomRangeFloat(Minimum=1, Maximum=40)
    )
    Index 2 Param 1=&FloatFromCurve(FloatCurve=DataInterface() {propertyValues {Curve {keys [
         {interpMode=RCIM_Cubic, leaveTangent=-0.0025, tangentMode=RCTM_Break, time=0, value=0.5}
         {arriveTangent=-0.0316, interpMode=RCIM_Cubic, tangentMode=RCTM_Break, time=1, value=-0.2}
        ]}}})
    Index 2 Param 2=&FloatFromCurve(
     FloatCurve=DataInterface() {propertyValues {Curve {keys [
         {interpMode=RCIM_Cubic, tangentMode=RCTM_User, time=0, value=0}
         {arriveTangent=2.0522, interpMode=RCIM_Cubic, leaveTangent=2.0522, tangentMode=RCTM_User, time=1, value=1}
        ]}}}
     Scale Curve=2
    )
    Module.Index0Param0Write=true
    Module.Index0Param1Write=true
    Module.Index0Param2Write=true
    Module.Index0Param3Write=true
    Module.Index1Param0Write=true
    Module.Index1Param1Write=true
    Module.Index1Param2Write=true
    Module.Index1Param3Write=true
    Module.Index2Param0Write=true
    Module.Index2Param1Write=true
    Module.Index2Param2Write=false
    Module.Index2Param3Write=false
    Module.Index3Param0Write=false
    Module.Index3Param1Write=false
    Module.Index3Param2Write=false
    Module.Index3Param3Write=true
    Write Parameter Index 1=true
    Write Parameter Index 2=true
    Write Parameter Index 3=true
   )
   &ScaleColor(
    Scale Alpha=&FloatFromCurve(FloatCurve=DataInterface() {propertyValues {Curve {keys [
         {tangentMode=RCTM_User, time=0, value=1}
         {arriveTangent=-0.0405, interpMode=RCIM_Cubic, leaveTangent=-0.0405, tangentMode=RCTM_User, time=0.7009, value=1}
         {arriveTangent=-8.0089, interpMode=RCIM_Cubic, leaveTangent=-8.0089, tangentMode=RCTM_User, time=1, value=0}
        ]}}})
    Scale RGB=&Lerp_Vector(
     A=&VectorFromFloat(Value=&FloatFromCurve(
       FloatCurve=DataInterface() {propertyValues {Curve {keys [<at 0 1>,<at 0.1 1>,<at 0.2 2>,<at 0.3 1>,<at 0.6 1>,<at 0.7 2>,<at 0.8 1>,<at 1 1>]}}}
       Scale Curve=10
      ))
     B=&VectorFromFloat(Value=&RandomRangeFloat(Minimum=0.3, Maximum=5, Evaluation Type=::NewEnumerator1))
     Alpha=&FloatFromCurve(FloatCurve=DataInterface() {propertyValues {Curve {keys [
          {interpMode=RCIM_Cubic, leaveTangent=2.5294, tangentMode=RCTM_Break, time=0, value=0}
          {arriveTangent=-0.2166, interpMode=RCIM_Cubic, tangentMode=RCTM_Break, time=0.4, value=1}
         ]}}})
    )
    ScaleA=true
    ScaleRGB=true
   )
  ]
  rendererClasses [&NiagaraMeshRendererProperties]
  renderers [
   {rendererClass=&NiagaraMeshRendererProperties, rendererIndex=0}
   {rendererClass=&NiagaraMeshRendererProperties, rendererIndex=1}
  ]
 }
 Smoke {
  emitterUpdateScript [
   &EmitterState(Life Cycle Mode=::NewEnumerator1, Loop Behavior=::NewEnumerator1, Loop Duration=4, UseLoopDelay=false)
   &SpawnBurst_Instantaneous(Spawn Group=1, Use Spawn Probability=false, UseLoopCountLimit=false)
   &SpawnBurst_Instantaneous(Spawn Group=1, Spawn Time=1, Use Spawn Probability=false, UseLoopCountLimit=false) {module=SpawnBurst_Instantaneous001}
   &SpawnBurst_Instantaneous(Spawn Group=1, Spawn Time=2.4, Use Spawn Probability=false, UseLoopCountLimit=false) {module=SpawnBurst_Instantaneous002}
   &SpawnRate(SpawnRate=&Multiply_Float(A=20, B=$System.3Pulse), Use Spawn Probability=false)
  ]
  particleSpawnScript [
   &InitializeParticle(
    Color=<rgba 1 1 1 2>
    Sprite Rotation Mode=::NewEnumerator2
    Sprite Size Mode=::NewEnumerator1
    Uniform Sprite Size Max=139.6005
    Uniform Sprite Size Min=79.959
    UsePositionOffset=true
    Write Material Random=true
   )
   &CameraOffset(Camera Offset Amount=10)
   &ShapeLocation(
    Box Size=<100 10 10>
    MaskSpawnGroup=false
    Offset=<0 50 -90>
    Offset Mode=::NewEnumerator0
    Override Seed=false
    Shape Primitive=::NewEnumerator2
    Surface Only=false
    UseOwnerScale=true
   )
   &AddVelocity(
    Rotation Coordinate Space=::World
    Velocity=&RandomRangeVector(Minimum=<0 0 0.2>, Maximum=<0 0 1>)
    Velocity Speed Scale=&FloatFromCurve(
     FloatCurve=DataInterface() {propertyValues {Curve {keys [{interpMode=RCIM_Cubic, tangentMode=RCTM_User, time=0, value=0.5347}, {interpMode=RCIM_Cubic, tangentMode=RCTM_User, time=1, value=1}]}}}
     CurveIndex=$Emitter.NormalizedLoopAge
     Scale Curve=40
    )
   )
   /SetVariables(Particles.Smoke Lifetime:NiagaraFloat=&RandomRangeFloat(Minimum=0.5, Maximum=0.8)) {isSetParameters=true}
  ]
  particleUpdateScript [
   &ParticleState()
   /SetVariables(Particles.Update Smoke Age:NiagaraFloat=&Divide_Float(A=$Particles.Age, B=$Particles.Smoke Lifetime)) {isSetParameters=true}
   &SpriteRotationRate(Rotation Rate=&RandomRangeFloat(Minimum=-50, Maximum=50))
   &ScaleSpriteSize(
    Scale Sprite Size Mode=::NewEnumerator0
    Uniform Scale Factor=&FloatFromCurve(FloatCurve=DataInterface() {propertyValues {Curve {keys [<at 0 0.8>,<at 1 1>]}}}, CurveIndex=$Particles.Update Smoke Age, Scale Curve=0.45)
   )
   &Drag(Drag=0.3, Use Linear Drag=true, Use Rotational Drag=true)
   &SubUVAnimation(
    End Frame Range Override=30
    "Sprite Renderer "=DataInterface() {propertyValues {SpriteRenderer=&self:Smoke_0.NiagaraSpriteRendererProperties_0}}
    Start Frame Range Override=&RandomRangeInt(Minimum=0, Maximum=5)
    SubUV Lookup Index:NiagaraFloat=$Particles.Update Smoke Age
    UseEndFrame=true
    UseStartFrame=true
   )
   &ScaleColor(
    Scale RGB=&MakeVectorFromLinearColorRGB(Color=&ColorFromCurve(
      ColorCurve=DataInterface() {propertyValues {
        AlphaCurve {keys []}
        BlueCurve {keys [
          <at 0.008 0.1032>
          {arriveTangent=0.2504, interpMode=RCIM_Cubic, leaveTangent=0.2504, time=0.2544, value=0.1535}
          <at 0.9908 0.3493 RCIM_Cubic>
         ]}
        GreenCurve {keys [<at 0.008 0.1329>, {arriveTangent=0.342, interpMode=RCIM_Cubic, leaveTangent=0.342, time=0.2544, value=0.1744}, <at 0.9908 0.469 RCIM_Cubic>]}
        RedCurve {keys [
          <at 0.008 0.1383>
          {arriveTangent=0.3533, interpMode=RCIM_Cubic, leaveTangent=0.3533, time=0.2544, value=0.1753}
          <at 0.9908 0.4855 RCIM_Cubic>
         ]}
       }}
      CurveIndex=$Particles.Update Smoke Age
     ))
    ScaleA=true
    ScaleRGB=true
   ) {module=ScaleColor001}
   &ScaleColor(
    Scale Alpha=&Multiply_Float(
     A=&RandomRangeFloat(Minimum=0.5, Maximum=1)
     B=&FloatFromCurve(
      FloatCurve=DataInterface() {propertyValues {Curve {keys [
          {interpMode=RCIM_Cubic, tangentMode=RCTM_User, time=0, value=0}
          {arriveTangent=0.579, interpMode=RCIM_Cubic, leaveTangent=0.579, tangentMode=RCTM_User, time=0.35, value=0.85}
          {interpMode=RCIM_Cubic, tangentMode=RCTM_User, time=0.4997, value=0.9}
          {tangentMode=RCTM_User, time=1, value=0}
         ]}}}
      CurveIndex=$Particles.Update Smoke Age
      Scale Curve=&Multiply_Float(
       A=&FN_FrontEndFloatModifier(Gameplay=0.8, Front End=1)
       B=&FNTimeOfDayScalar(Night Scale=0.5, Morning Scale=0.6, Day Scale=0.6, Evening Scale=0.7)
      )
     )
    )
    Scale RGB=&MakeVectorFromLinearColorRGB(Color=&FNTimeOfDayColor(
      Night Color=<rgba 0.1479 0.4904 2 1>
      Morning Color=<rgba 0.3892 0.6717 2 1>
      Day Color=<rgba 0.3892 0.6717 2 1>
      Evening Color=<rgba 0.3892 0.6717 2 1>
     ))
    ScaleA=true
    ScaleRGB=true
   )
   &SolveForcesAndVelocity(Clamp Velocity=false, Limit Acceleration=false)
   &FN_Light_Attributes(
    Color=<rgba 0 0.527 5 1>
    Create Color Attribute=true
    Create Exponent Attribute=true
    Create Position Attribute=true
    Enable Visibility Tag=true
    Exponent=1
    Position=<0 40 -40>
    Scale Color=&FloatFromCurve(
     FloatCurve=DataInterface() {propertyValues {Curve {
        defaultValue=1
        keys [
         {arriveTangent=4.0325, interpMode=RCIM_Cubic, leaveTangent=4.0325, tangentMode=RCTM_User, time=0, value=0}
         {
          arriveTangent=-0.0425
          arriveTangentWeight=0.4103
          interpMode=RCIM_Cubic
          leaveTangentWeight=0.2153
          tangentMode=RCTM_Break
          tangentWeightMode=RCTWM_WeightedBoth
          time=0.5
          value=1
         }
         {arriveTangent=-2, interpMode=RCIM_Cubic, leaveTangent=-2, time=0.9275, value=0.7958}
         {arriveTangent=-6.1578, interpMode=RCIM_Cubic, leaveTangent=-6.1578, tangentMode=RCTM_User, time=1, value=0}
        ]
       }}}
     Scale Curve=&FN_FrontEndFloatModifier(Gameplay=3, Front End=0)
    )
    VisibilityTag:NiagaraInt32=&SpawnGroup()
   )
  ]
  rendererClasses [&NiagaraSpriteRendererProperties, &NiagaraLightRendererProperties]
  renderers [
   {rendererClass=&NiagaraSpriteRendererProperties}
   {rendererClass=&NiagaraLightRendererProperties, rendererIndex=1}
  ]
 }
}
summary {
 systemName=NS_Emote_ClawPad_Host_Lightning_Hero
 emitters [
  {emitterName=Main_Lightning, rendererClasses [&NiagaraMeshRendererProperties]}
  {
   emitterName=GroundImpactsLightning
   rendererClasses [&NiagaraMeshRendererProperties, &NiagaraSpriteRendererProperties]
  }
  {emitterName=Small_Branches, rendererClasses [&NiagaraMeshRendererProperties]}
  {emitterName=Directional_sparks, rendererClasses [&NiagaraSpriteRendererProperties]}
  {emitterName=Smoke, rendererClasses [&NiagaraSpriteRendererProperties, &NiagaraLightRendererProperties]}
 ]
 userVariables [
  User.bIsFrontEnd {type=NiagaraBool}
  User.bIsFrontEndPreview {type=NiagaraBool}
  User.HitGround {type=NiagaraBool, defaultValue=true}
 ]
}
